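# Translucent unlit Niagara sprite material that draws a segmented audio spectrum meter for the Cactus Dancer backpack.
# UV math builds a box grid, a stepped reveal from SpectrumSample and a PeakSample marker box, with dynamic parameters supplying the audio.
# Color hue-shifts red along V, with brightness driven by the sample and time of day, blending from particle color via IncomingAudioGate.
# Opacity combines masks, particle alpha and a sharp inverse-Fresnel fade.
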Material /BRCosmetics/Accessories/FORT_Backpacks/Backpack_M_MED_Cactus_Dancer/FX/M_AudioSpectrumBlocks
graph {
 properties {bUsedWithNiagaraSprites=true, blendMode=BLEND_Translucent, pixelDepthOffsetMode=PDOM_Legacy, shadingModel=MSM_Unlit, usedShadingModels=Unlit}
 nodes [
  n1 = ParticleColor()
  n2 = ScalarParameter() {defaultValue=125, parameterName=BrightnessHigh}
  n3 = Multiply(A=@n2) {constB=0.6}
  n4 = Multiply(A=@n2) {constB=0.3}
  n5 = DynamicParameter() {defaultValue=<rgba 0 0 1 1>, paramNames [SpectrumSample, PeakSample, IncomingAudioGate, OverallAmplitude]}
  n6 = TextureCoordinate()
  n7 = Add(
   A=Multiply(A=ComponentMask(None=@n6) {g=true}, B=ScalarParameter() {defaultValue=1.5, parameterName=ColorRampMult})
   B=ScalarParameter() {defaultValue=-0.7, parameterName=ColorRampSlide}
  )
  n8 = Multiply(
   A=LinearInterpolate(
    A=ScalarParameter() {defaultValue=0.2, parameterName=BrightnessLow}
    B=MaterialFunctionCall(Value_Morning=@n3, Value_Day=@n2, Value_Evening=@n3, Value_Night=@n4) {materialFunction=&Scalar_Time_of_Day}.Result
    Alpha=Power(Base=@n5.SpectrumSample) {constExponent=5}
   )
   B=MaterialFunctionCall(Hue Shift Percentage=Saturate(None=@n7), Texture=Constant3Vector() {constant=<rgba 1 0 0 0>}) {materialFunction=&HueShift}.Result
  )
  n9 = DynamicParameter() {defaultValue=<rgba 0 0 1 1>, paramNames [SpectrumSample, PeakSample, IncomingAudioGate, OverallAmplitude]}
  MaterialFunctionCall(Emissive=LinearInterpolate(A=LinearInterpolate(A=@n1.RGB, B=@n8, Alpha=@n9.IncomingAudioGate), B=@n1.RGB, Alpha=NamedRerouteUsage())) {materialFunction=&MF_FX_EmissiveAdjustments}
  Multiply(A=Divide(B=ScalarParameter() {defaultValue=12, parameterName=BoxCount}))
  n10 = DynamicParameter() {defaultValue=<rgba 0 0 1 1>, paramNames [SpectrumSample, PeakSample, IncomingAudioGate, OverallAmplitude]}
  n11 = Add(A=Multiply(A=NamedRerouteUsage(), B=NamedRerouteUsage()), B=Multiply(A=NamedRerouteUsage(), B=@n10.IncomingAudioGate))
  n12 = OneMinus(None=Fresnel() {baseReflectFraction=0, exponent=15})
  Multiply(A=Multiply(A=@n1.A, B=Saturate(None=@n11)), B=Saturate(None=Power(Base=@n12) {constExponent=50}))
  n13 = TextureCoordinate()
  n14 = Frac(None=Multiply(A=ComponentMask(None=@n13) {g=true}, B=ScalarParameter() {defaultValue=12, parameterName=BoxCount}))
  n15 = ScalarParameter() {defaultValue=0.15, parameterName=EdgeWidth_Vert}
  n16 = Constant() {r=1}
  n17 = If(A=@n14, B=@n15, "A > B"=@n16, "A == B"=@n16, "A < B"=Constant())
  n18 = Constant() {r=1}
  n19 = If(A=@n14, B=OneMinus(None=@n15), "A > B"=Constant(), "A == B"=@n18, "A < B"=@n18)
  n20 = Frac(None=ComponentMask(None=TextureCoordinate() {vTiling=8}) {r=true})
  n21 = ScalarParameter() {defaultValue=0.2, parameterName=EdgeWidth_Horiz}
  n22 = Constant() {r=1}
  n23 = If(A=@n20, B=@n21, "A > B"=@n22, "A == B"=@n22, "A < B"=Constant())
  n24 = Constant() {r=1}
  n25 = If(A=@n20, B=OneMinus(None=@n21), "A > B"=Constant(), "A == B"=@n24, "A < B"=@n24)
  NamedRerouteDeclaration(None=Multiply(A=Multiply(A=@n17, B=@n19), B=Multiply(A=@n23, B=@n25))) {name=BoxMasks, desc=Declare Variable, nodeColor=<rgba 0.3828 1 0 1>}
  n26 = ComponentMask(None=TextureCoordinate()) {g=true}
  n27 = ScalarParameter() {defaultValue=12, parameterName=BoxCount}
  n28 = Floor(None=Multiply(A=OneMinus(None=@n26), B=@n27))
  n29 = DynamicParameter() {defaultValue=<rgba 0 0 1 1>, paramNames [SpectrumSample, PeakSample, IncomingAudioGate, OverallAmplitude]}
  n30 = Floor(None=Multiply(A=@n27, B=Multiply(A=@n29.SpectrumSample) {constB=0.8}))
  n31 = Constant()
  NamedRerouteDeclaration(None=If(A=Divide(A=@n28, B=@n27), B=Divide(A=@n30, B=@n27), "A > B"=@n31, "A == B"=@n31, "A < B"=Constant() {r=1})) {name=RevealMask, nodeColor=<rgba 1 0 0.7969 1>}
  n32 = ComponentMask(None=TextureCoordinate()) {g=true}
  n33 = DynamicParameter() {defaultValue=<rgba 0 0 1 1>, paramNames [SpectrumSample, PeakSample, IncomingAudioGate, OverallAmplitude]}
  n34 = Multiply(A=@n33.PeakSample) {constB=0.9}
  n35 = Multiply(A=OneMinus(None=Add(A=@n32, B=@n34)), B=ScalarParameter() {defaultValue=12, parameterName=BoxCount})
  n36 = ScalarParameter() {defaultValue=0.15, parameterName=PeakEdgeWidth_Vert}
  n37 = Constant() {r=1}
  n38 = If(A=@n35, B=@n36, "A > B"=@n37, "A == B"=@n37, "A < B"=Constant())
  n39 = Constant() {r=1}
  n40 = If(A=@n35, B=OneMinus(None=@n36), "A > B"=Constant(), "A == B"=@n39, "A < B"=@n39)
  n41 = Frac(None=ComponentMask(None=TextureCoordinate() {vTiling=8}) {r=true})
  n42 = ScalarParameter() {defaultValue=0.2, parameterName=PeakEdgeWidth_Horiz}
  n43 = Constant() {r=1}
  n44 = If(A=@n41, B=@n42, "A > B"=@n43, "A == B"=@n43, "A < B"=Constant())
  n45 = Constant() {r=1}
  n46 = If(A=@n41, B=OneMinus(None=@n42), "A > B"=Constant(), "A == B"=@n45, "A < B"=@n45)
  NamedRerouteDeclaration(None=Multiply(A=Multiply(A=@n38, B=@n40), B=Multiply(A=@n44, B=@n46))) {name=PeakBoxMask, nodeColor=<rgba 0.1719 1 0 1>}
  VectorParameter() {parameterName=BelowTheLine, defaultValue=<rgba 0 1 0 0>}
 ]
}
