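# Audio-reactive Niagara system for the Cactus Dancer backpack, drawing spectrum bars, a waveform ribbon and a pulsing colored light.
# Bar and ribbon emitters blend synthetic idle waves into the four-band event audio spectrum using an audio gate.
# Left and right bars interpolate bands by position, passing samples and decaying peaks to dynamic material parameters; overall amplitude drives ribbon height.
# Disabled GPU single-band variants include propagating amplitude between particles via particle reads.
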NiagaraSystem /BRCosmetics/Accessories/FORT_Backpacks/Backpack_M_MED_Cactus_Dancer/FX/NS_Backpack_CactusDancer
dependencies {
 usedDataInterfaces [
  &NiagaraDataInterfaceVectorCurve
  &NiagaraDataInterfaceColorCurve
  &NiagaraDataInterfaceCurve
  &NiagaraDataInterfaceParticleRead
 ]
 usedDynamicInputs []
 usedModules [
  &self:SystemSpawnScript.SystemScriptSource.SystemScriptGraph.NiagaraNodeAssignment_0.SetVariables_#1
  &SystemState
  &self:SystemSpawnScript.SystemScriptSource.SystemScriptGraph.NiagaraNodeAssignment_1.SetVariables_#2
 ]
}
emitters {
 4BandL {
  emitterUpdateScript [
   &EmitterState(Life Cycle Mode=::NewEnumerator1, Loop Behavior=::NewEnumerator1, Loop Duration Mode=::NewEnumerator1, UseLoopDelay=false)
   &SpawnBurst_Instantaneous(Spawn Count=9, Use Spawn Probability=false, UseLoopCountLimit=false)
   /SetVariables(Emitter.IncomingSpectrumData:LinearColor=&LerpLinearColors(
     StartColor=&MakeLinearColor(
      R:NiagaraFloat=&Sine(Normalized Angle=$Emitter.Age, Period=4, Scale=0.1, Bias=0.4)
      G:NiagaraFloat=&Cosine(Normalized Angle=$Emitter.Age, Period=2.8, Scale=0.1, Bias=0.4)
      B:NiagaraFloat=&Sine(Normalized Angle=$Emitter.Age, Period=3.25, Scale=0.1, Bias=0.3)
      A:NiagaraFloat=&Cosine(Normalized Angle=$Emitter.Age, Period=4, Scale=0.1, Bias=0.4)
     )
     EndColor=/Niagara/DynamicInputs/LinearColor/MakeLinearColorFromFloat4(Vector:Vector4f=&MakeVector4(
       X=&OneMinusFloat(Float=&Power(
         a=&OneMinusFloat(Float=&MakeFloatFromLinearColor(LinearColor=$NPC.FortniteNPC.EventAudio_Spectrum, Channel:ENiagaraLinearColor_Channels=::NewEnumerator0))
         b=0.5
        ))
       Y=&OneMinusFloat(Float=&Power(
         a=&OneMinusFloat(Float=&MakeFloatFromLinearColor(LinearColor=$NPC.FortniteNPC.EventAudio_Spectrum, Channel:ENiagaraLinearColor_Channels=::NewEnumerator1))
         b=0.75
        ))
       Z=&OneMinusFloat(Float=&Power(
         a=&OneMinusFloat(Float=&MakeFloatFromLinearColor(LinearColor=$NPC.FortniteNPC.EventAudio_Spectrum, Channel:ENiagaraLinearColor_Channels=::NewEnumerator2))
         b=1
        ))
       W=&OneMinusFloat(Float=&Power(
         a=&OneMinusFloat(Float=&MakeFloatFromLinearColor(LinearColor=$NPC.FortniteNPC.EventAudio_Spectrum, Channel:ENiagaraLinearColor_Channels=::NewEnumerator3))
         b=1.2
        ))
      ))
     LerpFactor=$System.IncomingAudioGate
    )) {isSetParameters=true, module=SetVariables_#3}
  ]
  particleSpawnScript [
   /SetVariables(Particles.SpectrumPosition:NiagaraFloat=&ReturnNormalizedExecIndex()) {isSetParameters=true, module=SetVariables_#4}
   &InitializeParticle(
    Lifetime=25
    Mesh Scale=<2 2 2>
    Mesh Scale Mode=::NewEnumerator4
    Position Offset=&VectorFromCurve(
     VectorCurve=DataInterface() {propertyValues {
       XCurve {keys [
         {arriveTangent=1.9015, interpMode=RCIM_Cubic, leaveTangent=1.9015, tangentMode=RCTM_User, time=0, value=0}
         <at 0.5 0.5 RCIM_Cubic>
         {arriveTangent=-1.9808, interpMode=RCIM_Cubic, leaveTangent=-1.9808, tangentMode=RCTM_User, time=1, value=0}
        ]}
       YCurve {keys [<at 0 0>,<at 1 0>]}
       ZCurve {keys [
         {
          arriveTangent=-0.0114
          arriveTangentWeight=0.4956
          interpMode=RCIM_Cubic
          leaveTangent=-0.0114
          leaveTangentWeight=0.4956
          tangentMode=RCTM_User
          tangentWeightMode=RCTWM_WeightedBoth
          time=0
          value=-1
         }
         {
          arriveTangent=-0.0012
          arriveTangentWeight=0.4997
          interpMode=RCIM_Cubic
          leaveTangent=-0.0012
          leaveTangentWeight=0.4997
          tangentMode=RCTM_User
          tangentWeightMode=RCTWM_WeightedBoth
          time=1
          value=1
         }
        ]}
      }}
     CurveIndex=$Particles.SpectrumPosition
     Scale Curve=<31 1 13.2>
    )
    Sprite Size=<3 28>
    Sprite Size Mode=::NewEnumerator4
    UsePositionOffset=false
    Write Material Random=true
   )
   &ShapeLocation(
    Handle Radius=0
    Large Radius=16
    MaskSpawnGroup=false
    Rotation Mode=::NewEnumerator2
    Shape Primitive=::NewEnumerator3
    Torus Distribution Mode:ENiagaraTorusDistributionMode=::NewEnumerator1
    U Position=&Add_Float(A=&Multiply_Float(A=$Particles.SpectrumPosition, B=0.2), B=0.15)
    UseOwnerScale=true
    V Position=0
    Yaw / Pitch / Roll=<0.5 0 -0.25>
   )
   &InitialMeshOrientation(
    Mesh Orientation Mode=::NewEnumerator2
    Orientation Vector=&Subtract_Vector(
     A=&ConvertPositionToVector(Input Position=$Particles.Position, Convert To Absolute Large World Vector=false)
     B=<0 0 -5>
    )
    Use Rotation Vector=false
   )
   /SetVariables() {isSetParameters=true, module=SetVariables_#5}
   &AlignSpriteToMeshOrientation(Mesh Orientation Relative Sprite Alignment Vector=<1 0 0>, Mesh Orientation Relative Sprite Facing Vector:Vector3f=<-1 0 0>)
  ]
  particleUpdateScript [
   &ParticleState(Kill Particles When Lifetime Has Elapsed=false)
   &UpdateMeshOrientation(
    Rotation Rate=&Sine(Normalized Angle=&Add_Float(A=$Emitter.Age, B=$Particles.SpectrumPosition), Period=0.333, Scale=0.01, Bias=0)
    Rotation Vector=<0 0 1>
   )
   /SetVariables(Particles.SpectrumSample:NiagaraFloat=&Add_Float(
     A=&self:MaxOfColorChannels(Color:LinearColor=&MultiplyLinearColor(
       Linear Color=$Emitter.IncomingSpectrumData
       Scale Factor=&MakeVector4_FromLinearColor(LinearColor=&ColorFromCurve(
         ColorCurve=DataInterface() {propertyValues {
           AlphaCurve {keys [
             <at 0 0 RCIM_Cubic>
             {arriveTangentWeight=0.3202, interpMode=RCIM_Cubic, leaveTangentWeight=0.1067, tangentMode=RCTM_User, time=0.6, value=0}
             <at 0.7 0.7>
             {arriveTangentWeight=0.1438, interpMode=RCIM_Cubic, leaveTangentWeight=0.1438, tangentMode=RCTM_User, time=0.8, value=1}
             <at 1 0.5 RCIM_Cubic>
            ]}
           BlueCurve {keys [
             <at 0 0 RCIM_Cubic>
             {arriveTangentWeight=0.0667, interpMode=RCIM_Cubic, leaveTangentWeight=0.0667, tangentMode=RCTM_User, time=0.2, value=0}
             <at 0.2998 0>
             {arriveTangentWeight=0.1795, interpMode=RCIM_Cubic, leaveTangentWeight=0.1795, tangentMode=RCTM_User, time=0.4, value=0}
             <at 0.5 0.7>
             {
              arriveTangentWeight=0.1547
              interpMode=RCIM_Cubic
              leaveTangentWeight=0.1547
              tangentMode=RCTM_User
              tangentWeightMode=RCTWM_WeightedBoth
              time=0.6
              value=1
             }
             <at 0.7 0.7>
             {arriveTangentWeight=0.0667, leaveTangentWeight=0.0667, tangentMode=RCTM_User, time=0.8, value=0}
             <at 1 0>
            ]}
           GreenCurve {keys [
             <at 0 0 RCIM_Cubic>
             {arriveTangentWeight=0.1795, interpMode=RCIM_Cubic, leaveTangentWeight=0.1795, tangentMode=RCTM_User, time=0.2, value=0}
             <at 0.2998 0.7>
             {
              arriveTangentWeight=0.1699
              interpMode=RCIM_Cubic
              leaveTangentWeight=0.1699
              tangentMode=RCTM_User
              tangentWeightMode=RCTWM_WeightedBoth
              time=0.4
              value=1
             }
             <at 0.5 0.7>
             {arriveTangentWeight=0.1795, interpMode=RCIM_Cubic, leaveTangentWeight=0.1795, tangentMode=RCTM_User, time=0.6, value=0}
             <at 0.7 0>
             {arriveTangentWeight=0.0667, leaveTangentWeight=0.0667, tangentMode=RCTM_User, time=0.8, value=0}
             <at 1 0>
            ]}
           RedCurve {keys [
             <at 0 0.5 RCIM_Cubic>
             {arriveTangentWeight=0.1342, interpMode=RCIM_Cubic, leaveTangentWeight=0.1342, tangentMode=RCTM_User, time=0.2, value=1}
             <at 0.2998 0.7>
             {arriveTangentWeight=0.1795, interpMode=RCIM_Cubic, leaveTangentWeight=0.1795, tangentMode=RCTM_User, time=0.4, value=0}
             <at 0.5 0>
             {arriveTangentWeight=0.0667, interpMode=RCIM_Cubic, leaveTangentWeight=0.0667, tangentMode=RCTM_User, time=0.6, value=0}
             <at 0.7 0>
             <at 0.8 0>
             <at 1 0>
            ]}
          }}
         CurveIndex=$Particles.SpectrumPosition
        ))
      ))
     B=&Waveform(
      Global Phase=&Add_Float(A=$Emitter.Age, B=$Particles.MaterialRandom)
      "Waveform [1]"=::NewEnumerator0
      "[1] Frequency"=2
      "[1] Compound Frequency"=3.17
      "[1] Amplitude Min/Max"=<0 0.1>
     )
    )) {isSetParameters=true, module=SetVariables_#6}
   /SetVariables(Particles.PeakSample:NiagaraFloat=&MakeCustomFloatFromBool(
     Bool=&SetBoolByFloatComparison(A=$Particles.SpectrumSample, B=$Particles.PeakSample)
     True Float=$Particles.SpectrumSample
     False Float=&ClampFloat(Float=&Subtract_Float(A=$Particles.PeakSample, B=&Multiply_Float(A=$System.PeakDecayRate, B=$Engine.DeltaTime)))
    )) {isSetParameters=true, module=SetVariables_#7}
   &Color(Color=&ColorFromCurve(
     ColorCurve=DataInterface() {propertyValues {
       AlphaCurve {keys [<at 0 1 RCIM_Cubic>]}
       BlueCurve {keys [<at 0 0 RCIM_Cubic>,<at 0.4903 1>,<at 1 0.2375 RCIM_Cubic>]}
       GreenCurve {keys [<at 0 0 RCIM_Cubic>,<at 0.4903 0.0787>,<at 1 1 RCIM_Cubic>]}
       RedCurve {keys [<at 0 1 RCIM_Cubic>,<at 0.4903 0>,<at 1 0 RCIM_Cubic>]}
      }}
     CurveIndex=&Sine(Normalized Angle=&Add_Float(A=$Emitter.Age, B=$Particles.SpectrumPosition), Period=1, Scale=0.5, Bias=0.5)
    ))
   &ScaleSpriteSize(
    Scale Factor=&Lerp_Vector2(A=<1 1>, B=<1 0.85>, Alpha=$Particles.SpectrumSample)
    Scale Sprite Size Mode=::NewEnumerator2
   )
   &AlignSpriteToMeshOrientation(Mesh Orientation Relative Sprite Alignment Vector=<1 0 0>, Mesh Orientation Relative Sprite Facing Vector:Vector3f=<-1 0 0>) {module=AlignSpriteToMeshOrientation001}
   &DynamicMaterialParameters(
    Index 0 Param 1=$Particles.SpectrumSample
    Index 0 Param 2=$Particles.PeakSample
    Index 0 Param 3=$System.IncomingAudioGate
    Module.Param0WriteEnabled=true
    Module.Param1WriteEnabled=true
    Module.Param2WriteEnabled=true
    Module.Param3WriteEnabled=true
   )
  ]
  rendererClasses [&NiagaraSpriteRendererProperties]
  renderers [{rendererClass=&NiagaraSpriteRendererProperties}]
 }
 4BandR {
  emitterUpdateScript [
   &EmitterState(Life Cycle Mode=::NewEnumerator1, Loop Behavior=::NewEnumerator1, Loop Duration Mode=::NewEnumerator1, UseLoopDelay=false)
   &SpawnBurst_Instantaneous(Spawn Count=9, Use Spawn Probability=false, UseLoopCountLimit=false)
   /SetVariables(Emitter.IncomingSpectrumData:LinearColor=&LerpLinearColors(
     StartColor=&MakeLinearColor(
      R:NiagaraFloat=&Sine(Normalized Angle=$Emitter.Age, Period=4, Scale=0.1, Bias=0.4)
      G:NiagaraFloat=&Cosine(Normalized Angle=$Emitter.Age, Period=2.8, Scale=0.1, Bias=0.4)
      B:NiagaraFloat=&Sine(Normalized Angle=$Emitter.Age, Period=3.25, Scale=0.1, Bias=0.3)
      A:NiagaraFloat=&Cosine(Normalized Angle=$Emitter.Age, Period=4, Scale=0.1, Bias=0.4)
     )
     EndColor=&MakeLinearColorFromFloat4(Vector:Vector4f=&MakeVector4(
       X=&OneMinusFloat(Float=&Power(
         a=&OneMinusFloat(Float=&MakeFloatFromLinearColor(LinearColor=$NPC.FortniteNPC.EventAudio_Spectrum, Channel:ENiagaraLinearColor_Channels=::NewEnumerator0))
         b=0.5
        ))
       Y=&OneMinusFloat(Float=&Power(
         a=&OneMinusFloat(Float=&MakeFloatFromLinearColor(LinearColor=$NPC.FortniteNPC.EventAudio_Spectrum, Channel:ENiagaraLinearColor_Channels=::NewEnumerator1))
         b=0.75
        ))
       Z=&OneMinusFloat(Float=&Power(
         a=&OneMinusFloat(Float=&MakeFloatFromLinearColor(LinearColor=$NPC.FortniteNPC.EventAudio_Spectrum, Channel:ENiagaraLinearColor_Channels=::NewEnumerator2))
         b=1
        ))
       W=&OneMinusFloat(Float=&Power(
         a=&OneMinusFloat(Float=&MakeFloatFromLinearColor(LinearColor=$NPC.FortniteNPC.EventAudio_Spectrum, Channel:ENiagaraLinearColor_Channels=::NewEnumerator3))
         b=1.2
        ))
      ))
     LerpFactor=$System.IncomingAudioGate
    )) {isSetParameters=true, module=SetVariables_#3}
  ]
  particleSpawnScript [
   /SetVariables(Particles.SpectrumPosition:NiagaraFloat=&ReturnNormalizedExecIndex()) {isSetParameters=true, module=SetVariables_#4}
   &InitializeParticle(
    Lifetime=25
    Mesh Scale=<2 2 2>
    Mesh Scale Mode=::NewEnumerator4
    Position Offset=&VectorFromCurve(
     VectorCurve=DataInterface() {propertyValues {
       XCurve {keys [
         {arriveTangent=1.9015, interpMode=RCIM_Cubic, leaveTangent=1.9015, tangentMode=RCTM_User, time=0, value=0}
         <at 0.5 0.5 RCIM_Cubic>
         {arriveTangent=-1.9808, interpMode=RCIM_Cubic, leaveTangent=-1.9808, tangentMode=RCTM_User, time=1, value=0}
        ]}
       YCurve {keys [<at 0 0>,<at 1 0>]}
       ZCurve {keys [
         {
          arriveTangent=-0.0114
          arriveTangentWeight=0.4956
          interpMode=RCIM_Cubic
          leaveTangent=-0.0114
          leaveTangentWeight=0.4956
          tangentMode=RCTM_User
          tangentWeightMode=RCTWM_WeightedBoth
          time=0
          value=-1
         }
         {
          arriveTangent=-0.0012
          arriveTangentWeight=0.4997
          interpMode=RCIM_Cubic
          leaveTangent=-0.0012
          leaveTangentWeight=0.4997
          tangentMode=RCTM_User
          tangentWeightMode=RCTWM_WeightedBoth
          time=1
          value=1
         }
        ]}
      }}
     CurveIndex=$Particles.SpectrumPosition
     Scale Curve=<31 1 13.2>
    )
    Sprite Size=<3 28>
    Sprite Size Mode=::NewEnumerator4
    UsePositionOffset=false
    Write Material Random=true
   )
   &ShapeLocation(
    Handle Radius=0
    Large Radius=16
    MaskSpawnGroup=false
    Rotation Mode=::NewEnumerator2
    Shape Primitive=::NewEnumerator3
    Torus Distribution Mode:ENiagaraTorusDistributionMode=::NewEnumerator1
    U Position=&Add_Float(A=&Multiply_Float(A=$Particles.SpectrumPosition, B=-0.2), B=0.85)
    UseOwnerScale=true
    V Position=0
    Yaw / Pitch / Roll=<0.5 0 -0.25>
   )
   &InitialMeshOrientation(
    Mesh Orientation Mode=::NewEnumerator2
    Orientation Vector=&Subtract_Vector(
     A=&ConvertPositionToVector(Input Position=$Particles.Position, Convert To Absolute Large World Vector=false)
     B=<0 0 5>
    )
    Use Rotation Vector=false
   )
   /SetVariables() {isSetParameters=true, module=SetVariables_#5}
   &AlignSpriteToMeshOrientation(Mesh Orientation Relative Sprite Alignment Vector=<1 0 0>, Mesh Orientation Relative Sprite Facing Vector:Vector3f=<-1 0 0>)
  ]
  particleUpdateScript [
   &ParticleState(Kill Particles When Lifetime Has Elapsed=false)
   &UpdateMeshOrientation(
    Rotation Rate=&Sine(Normalized Angle=&Add_Float(A=$Emitter.Age, B=$Particles.SpectrumPosition), Period=0.333, Scale=0.01, Bias=0)
    Rotation Vector=<0 0 1>
   )
   /SetVariables(Particles.SpectrumSample:NiagaraFloat=&Add_Float(
     A=&self:MaxOfColorChannels(Color:LinearColor=&MultiplyLinearColor(
       Linear Color=$Emitter.IncomingSpectrumData
       Scale Factor=&MakeVector4_FromLinearColor(LinearColor=&ColorFromCurve(
         ColorCurve=DataInterface() {propertyValues {
           AlphaCurve {keys [
             <at 0 0 RCIM_Cubic>
             {arriveTangentWeight=0.3202, interpMode=RCIM_Cubic, leaveTangentWeight=0.1067, tangentMode=RCTM_User, time=0.6, value=0}
             <at 0.7 0.7>
             {arriveTangentWeight=0.1438, interpMode=RCIM_Cubic, leaveTangentWeight=0.1438, tangentMode=RCTM_User, time=0.8, value=1}
             <at 1 0.5 RCIM_Cubic>
            ]}
           BlueCurve {keys [
             <at 0 0 RCIM_Cubic>
             {arriveTangentWeight=0.0667, interpMode=RCIM_Cubic, leaveTangentWeight=0.0667, tangentMode=RCTM_User, time=0.2, value=0}
             <at 0.2998 0>
             {arriveTangentWeight=0.1795, interpMode=RCIM_Cubic, leaveTangentWeight=0.1795, tangentMode=RCTM_User, time=0.4, value=0}
             <at 0.5 0.7>
             {
              arriveTangentWeight=0.1547
              interpMode=RCIM_Cubic
              leaveTangentWeight=0.1547
              tangentMode=RCTM_User
              tangentWeightMode=RCTWM_WeightedBoth
              time=0.6
              value=1
             }
             <at 0.7 0.7>
             {arriveTangentWeight=0.0667, leaveTangentWeight=0.0667, tangentMode=RCTM_User, time=0.8, value=0}
             <at 1 0>
            ]}
           GreenCurve {keys [
             <at 0 0 RCIM_Cubic>
             {arriveTangentWeight=0.1795, interpMode=RCIM_Cubic, leaveTangentWeight=0.1795, tangentMode=RCTM_User, time=0.2, value=0}
             <at 0.2998 0.7>
             {
              arriveTangentWeight=0.1699
              interpMode=RCIM_Cubic
              leaveTangentWeight=0.1699
              tangentMode=RCTM_User
              tangentWeightMode=RCTWM_WeightedBoth
              time=0.4
              value=1
             }
             <at 0.5 0.7>
             {arriveTangentWeight=0.1795, interpMode=RCIM_Cubic, leaveTangentWeight=0.1795, tangentMode=RCTM_User, time=0.6, value=0}
             <at 0.7 0>
             {arriveTangentWeight=0.0667, leaveTangentWeight=0.0667, tangentMode=RCTM_User, time=0.8, value=0}
             <at 1 0>
            ]}
           RedCurve {keys [
             <at 0 0.5 RCIM_Cubic>
             {arriveTangentWeight=0.1342, interpMode=RCIM_Cubic, leaveTangentWeight=0.1342, tangentMode=RCTM_User, time=0.2, value=1}
             <at 0.2998 0.7>
             {arriveTangentWeight=0.1795, interpMode=RCIM_Cubic, leaveTangentWeight=0.1795, tangentMode=RCTM_User, time=0.4, value=0}
             <at 0.5 0>
             {arriveTangentWeight=0.0667, interpMode=RCIM_Cubic, leaveTangentWeight=0.0667, tangentMode=RCTM_User, time=0.6, value=0}
             <at 0.7 0>
             <at 0.8 0>
             <at 1 0>
            ]}
          }}
         CurveIndex=$Particles.SpectrumPosition
        ))
      ))
     B=&Waveform(
      Global Phase=&Add_Float(A=$Emitter.Age, B=$Particles.MaterialRandom)
      "Waveform [1]"=::NewEnumerator0
      "[1] Frequency"=2
      "[1] Compound Frequency"=3.17
      "[1] Amplitude Min/Max"=<0 0.1>
     )
    )) {isSetParameters=true, module=SetVariables_#6}
   /SetVariables(Particles.PeakSample:NiagaraFloat=&MakeCustomFloatFromBool(
     Bool=&SetBoolByFloatComparison(A=$Particles.SpectrumSample, B=$Particles.PeakSample)
     True Float=$Particles.SpectrumSample
     False Float=&ClampFloat(Float=&Subtract_Float(A=$Particles.PeakSample, B=&Multiply_Float(A=$System.PeakDecayRate, B=$Engine.DeltaTime)))
    )) {isSetParameters=true, module=SetVariables_#7}
   &Color(Color=&ColorFromCurve(
     ColorCurve=DataInterface() {propertyValues {
       AlphaCurve {keys [<at 0 1 RCIM_Cubic>]}
       BlueCurve {keys [<at 0 0 RCIM_Cubic>,<at 0.4903 1>,<at 1 0.2375 RCIM_Cubic>]}
       GreenCurve {keys [<at 0 0 RCIM_Cubic>,<at 0.4903 0.0787>,<at 1 1 RCIM_Cubic>]}
       RedCurve {keys [<at 0 1 RCIM_Cubic>,<at 0.4903 0>,<at 1 0 RCIM_Cubic>]}
      }}
     CurveIndex=&Sine(Normalized Angle=&Add_Float(A=$Emitter.Age, B=$Particles.SpectrumPosition), Period=1, Scale=0.5, Bias=0.5)
    ))
   &ScaleSpriteSize(
    Scale Factor=&Lerp_Vector2(A=<1 1>, B=<1 0.85>, Alpha=$Particles.SpectrumSample)
    Scale Sprite Size Mode=::NewEnumerator2
   )
   &AlignSpriteToMeshOrientation(Mesh Orientation Relative Sprite Alignment Vector=<1 0 0>, Mesh Orientation Relative Sprite Facing Vector:Vector3f=<-1 0 0>) {module=AlignSpriteToMeshOrientation001}
   &DynamicMaterialParameters(
    Index 0 Param 1=$Particles.SpectrumSample
    Index 0 Param 2=$Particles.PeakSample
    Index 0 Param 3=$System.IncomingAudioGate
    Module.Param0WriteEnabled=true
    Module.Param1WriteEnabled=true
    Module.Param2WriteEnabled=true
    Module.Param3WriteEnabled=false
   )
  ]
  rendererClasses [&NiagaraSpriteRendererProperties]
  renderers [{rendererClass=&NiagaraSpriteRendererProperties}]
 }
 Light {
  emitterUpdateScript [
   &EmitterState(Inactive Response=::NewEnumerator1, Life Cycle Mode=::NewEnumerator1, Loop Duration=4, UseLoopDelay=true)
   &SpawnBurst_Instantaneous(Spawn Count=1, Use Spawn Probability=false)
  ]
  particleSpawnScript [&InitializeParticle(Color Mode=::NewEnumerator0, Lifetime=4, Position Offset=<20 -20 0>, UsePositionOffset=true, Write Material Random=true) {module=InitializeParticle001}]
  particleUpdateScript [
   &ParticleState()
   &Color(Color=&MakeLinearColorFromVectorAndFloat(
     "Vector (RGB)"=&Multiply_Vector(
      A=&MakeVectorFromLinearColorRGB(Color=&LerpLinearColors(
        StartColor=&LerpLinearColors(
         StartColor=<rgba 0 0.3964 1 1>
         EndColor=<rgba 1 0 0.9749 1>
         LerpFactor=&FloatFromCurve(
          FloatCurve=DataInterface() {propertyValues {Curve {keys [<at 0 1>,<at 0.1069 0.0895>,<at 0.3206 0.7423>,<at 0.4907 0.1444>,<at 0.6992 0.8948>,<at 0.9061 0.1261>,<at 1 1>]}}}
          CurveIndex=$Emitter.NormalizedLoopAge
         )
        )
        EndColor=&HueShiftLinearColor(
         Hue=&FloatFromCurve(
          FloatCurve=DataInterface() {propertyValues {Curve {keys [<at 0 1>,<at 0.1069 0.0895>,<at 0.3206 0.7423>,<at 0.4907 0.1444>,<at 0.6992 0.8948>,<at 0.9061 0.1261>,<at 1 1>]}}}
          CurveIndex=&Frac_Float(Float=&Multiply_Float(A=$Emitter.NormalizedLoopAge, B=2))
         )
         Saturation=0.4
         Value=1
         Linear Color=&MakeLinearColor(R:NiagaraFloat=1, G:NiagaraFloat=0, B:NiagaraFloat=0, A:NiagaraFloat=0)
        )
        LerpFactor=$System.IncomingAudioGate
       ))
      B=&VectorFromFloat(Value=&FloatFromCurve(
        FloatCurve=DataInterface() {propertyValues {Curve {keys [<at 0 1>,<at 0.0584 1.1216>,<at 0.0797 0.6691>,<at 0.1464 0.526>,<at 0.222 1.0327>,<at 0.2705 0.6844>,<at 0.3166 0.8798>,<at 0.3641 0.5477>,<at 0.4407 0.6405>,<at 0.4815 1.1358>,<at 0.5388 1.0006>,<at 0.58 0.5314>,<at 0.6479 0.8407>,<at 0.6937 0.7459>,<at 0.737 1.0896>,<at 0.8125 0.5703>,<at 0.8909 0.8732>,<at 0.9449 0.6186>,<at 1 1>]}}}
        CurveIndex=$Emitter.NormalizedLoopAge
       ))
     )
     "Float (Alpha)"=1
    ))
   &FN_Light_v2(Light Exponent=2, Light Radius=50, LightEnabled=true, Scale Color=16)
  ]
  rendererClasses [&NiagaraLightRendererProperties]
  renderers [{rendererClass=&NiagaraLightRendererProperties, rendererIndex=0}]
 }
 Ribbon {
  emitterUpdateScript [
   &EmitterState(Life Cycle Mode=::NewEnumerator1, Loop Behavior=::NewEnumerator1, Loop Duration Mode=::NewEnumerator1, UseLoopDelay=false)
   &SpawnBurst_Instantaneous(Spawn Count=24, Use Spawn Probability=false, UseLoopCountLimit=false)
   /SetVariables(Emitter.IncomingSpectrumData:LinearColor=&LerpLinearColors(
     StartColor=&MakeLinearColor(
      R:NiagaraFloat=&Sine(Normalized Angle=$Emitter.Age, Period=4, Scale=0.1, Bias=0.4)
      G:NiagaraFloat=&Cosine(Normalized Angle=$Emitter.Age, Period=2.8, Scale=0.1, Bias=0.4)
      B:NiagaraFloat=&Sine(Normalized Angle=$Emitter.Age, Period=3.25, Scale=0.1, Bias=0.3)
      A:NiagaraFloat=&Cosine(Normalized Angle=$Emitter.Age, Period=4, Scale=0.1, Bias=0.4)
     )
     EndColor=&MakeLinearColorFromFloat4(Vector:Vector4f=&MakeVector4(
       X=&OneMinusFloat(Float=&Power(
         a=&OneMinusFloat(Float=&MakeFloatFromLinearColor(LinearColor=$NPC.FortniteNPC.EventAudio_Spectrum, Channel:ENiagaraLinearColor_Channels=::NewEnumerator0))
         b=0.5
        ))
       Y=&OneMinusFloat(Float=&Power(
         a=&OneMinusFloat(Float=&MakeFloatFromLinearColor(LinearColor=$NPC.FortniteNPC.EventAudio_Spectrum, Channel:ENiagaraLinearColor_Channels=::NewEnumerator1))
         b=0.75
        ))
       Z=&OneMinusFloat(Float=&Power(
         a=&OneMinusFloat(Float=&MakeFloatFromLinearColor(LinearColor=$NPC.FortniteNPC.EventAudio_Spectrum, Channel:ENiagaraLinearColor_Channels=::NewEnumerator2))
         b=1
        ))
       W=&OneMinusFloat(Float=&Power(
         a=&OneMinusFloat(Float=&MakeFloatFromLinearColor(LinearColor=$NPC.FortniteNPC.EventAudio_Spectrum, Channel:ENiagaraLinearColor_Channels=::NewEnumerator3))
         b=1.2
        ))
      ))
     LerpFactor=$System.IncomingAudioGate
    )) {isSetParameters=true, module=SetVariables_#3}
  ]
  particleSpawnScript [
   /SetVariables(Particles.SpectrumPosition:NiagaraFloat=&ReturnNormalizedExecIndex()) {isSetParameters=true, module=SetVariables_#4}
   /SetVariables(Particles.SpectrumUpVector:Vector3f=&VectorFromCurve(
     VectorCurve=DataInterface() {propertyValues {
       XCurve {keys [
         {arriveTangent=1.9015, interpMode=RCIM_Cubic, leaveTangent=1.9015, tangentMode=RCTM_User, time=0, value=0}
         <at 0.5 0.5 RCIM_Cubic>
         {arriveTangent=-1.9808, interpMode=RCIM_Cubic, leaveTangent=-1.9808, tangentMode=RCTM_User, time=1, value=0}
        ]}
       YCurve {keys [<at 0 0>,<at 1 0>]}
       ZCurve {keys [
         {
          arriveTangent=-0.0114
          arriveTangentWeight=0.4956
          interpMode=RCIM_Cubic
          leaveTangent=-0.0114
          leaveTangentWeight=0.4956
          tangentMode=RCTM_User
          tangentWeightMode=RCTWM_WeightedBoth
          time=0
          value=-1
         }
         {
          arriveTangent=-0.0012
          arriveTangentWeight=0.4997
          interpMode=RCIM_Cubic
          leaveTangent=-0.0012
          leaveTangentWeight=0.4997
          tangentMode=RCTM_User
          tangentWeightMode=RCTWM_WeightedBoth
          time=1
          value=1
         }
        ]}
      }}
     CurveIndex=$Particles.SpectrumPosition
     Scale Curve=<2.3 1 1>
    )) {isSetParameters=true, module=SetVariables_#8}
   &InitializeParticle(
    Color=<rgba 1 0 0.4592 1>
    Lifetime=25
    Mesh Scale=<2 2 2>
    Mesh Scale Mode=::NewEnumerator4
    Position Offset=&VectorFromCurve(
     VectorCurve=DataInterface() {propertyValues {
       XCurve {keys [
         {arriveTangent=1.9015, interpMode=RCIM_Cubic, leaveTangent=1.9015, tangentMode=RCTM_User, time=0, value=0}
         <at 0.5 0.5 RCIM_Cubic>
         {arriveTangent=-1.9808, interpMode=RCIM_Cubic, leaveTangent=-1.9808, tangentMode=RCTM_User, time=1, value=0}
        ]}
       YCurve {keys [<at 0 0>,<at 1 0>]}
       ZCurve {keys [
         {
          arriveTangent=-0.0114
          arriveTangentWeight=0.4956
          interpMode=RCIM_Cubic
          leaveTangent=-0.0114
          leaveTangentWeight=0.4956
          tangentMode=RCTM_User
          tangentWeightMode=RCTWM_WeightedBoth
          time=0
          value=-1
         }
         {
          arriveTangent=-0.0012
          arriveTangentWeight=0.4997
          interpMode=RCIM_Cubic
          leaveTangent=-0.0012
          leaveTangentWeight=0.4997
          tangentMode=RCTM_User
          tangentWeightMode=RCTWM_WeightedBoth
          time=1
          value=1
         }
        ]}
      }}
     CurveIndex=$Particles.SpectrumPosition
     Scale Curve=<31 1 13.2>
    )
    Sprite Size=<3 28>
    Sprite Size Mode=::NewEnumerator4
    UsePositionOffset=true
    Write Material Random=true
   )
   /SetVariables(Particles.RibbonLinkOrder:NiagaraFloat=$Particles.SpectrumPosition) {isSetParameters=true}
   /SetVariables(Particles.RibbonPosition:NiagaraPosition=$Particles.Position) {isSetParameters=true, module=SetVariables_#9}
   /SetVariables(Particles.RibbonPosition:NiagaraPosition=$Particles.Position) {isSetParameters=true}
  ]
  particleUpdateScript [
   &ParticleState(Kill Particles When Lifetime Has Elapsed=false)
   &UpdateMeshOrientation(
    Rotation Rate=&Sine(Normalized Angle=&Add_Float(A=$Emitter.Age, B=$Particles.SpectrumPosition), Period=0.333, Scale=0.01, Bias=0)
    Rotation Vector=<0 0 1>
   )
   /SetVariables(Particles.SpectrumSample:NiagaraFloat=&Add_Float(
     A=&self:MaxOfColorChannels(Color:LinearColor=&MultiplyLinearColor(
       Linear Color=$Emitter.IncomingSpectrumData
       Scale Factor=&MakeVector4_FromLinearColor(LinearColor=&ColorFromCurve(
         ColorCurve=DataInterface() {propertyValues {
           AlphaCurve {keys [
             <at 0 0 RCIM_Cubic>
             {arriveTangentWeight=0.3202, interpMode=RCIM_Cubic, leaveTangentWeight=0.1067, tangentMode=RCTM_User, time=0.6, value=0}
             <at 0.7 0.7>
             {arriveTangentWeight=0.1438, interpMode=RCIM_Cubic, leaveTangentWeight=0.1438, tangentMode=RCTM_User, time=0.8, value=1}
             <at 1 0.5 RCIM_Cubic>
            ]}
           BlueCurve {keys [
             <at 0 0 RCIM_Cubic>
             {arriveTangentWeight=0.0667, interpMode=RCIM_Cubic, leaveTangentWeight=0.0667, tangentMode=RCTM_User, time=0.2, value=0}
             <at 0.2998 0>
             {arriveTangentWeight=0.1795, interpMode=RCIM_Cubic, leaveTangentWeight=0.1795, tangentMode=RCTM_User, time=0.4, value=0}
             <at 0.5 0.7>
             {
              arriveTangentWeight=0.1547
              interpMode=RCIM_Cubic
              leaveTangentWeight=0.1547
              tangentMode=RCTM_User
              tangentWeightMode=RCTWM_WeightedBoth
              time=0.6
              value=1
             }
             <at 0.7 0.7>
             {arriveTangentWeight=0.0667, leaveTangentWeight=0.0667, tangentMode=RCTM_User, time=0.8, value=0}
             <at 1 0>
            ]}
           GreenCurve {keys [
             <at 0 0 RCIM_Cubic>
             {arriveTangentWeight=0.1795, interpMode=RCIM_Cubic, leaveTangentWeight=0.1795, tangentMode=RCTM_User, time=0.2, value=0}
             <at 0.2998 0.7>
             {
              arriveTangentWeight=0.1699
              interpMode=RCIM_Cubic
              leaveTangentWeight=0.1699
              tangentMode=RCTM_User
              tangentWeightMode=RCTWM_WeightedBoth
              time=0.4
              value=1
             }
             <at 0.5 0.7>
             {arriveTangentWeight=0.1795, interpMode=RCIM_Cubic, leaveTangentWeight=0.1795, tangentMode=RCTM_User, time=0.6, value=0}
             <at 0.7 0>
             {arriveTangentWeight=0.0667, leaveTangentWeight=0.0667, tangentMode=RCTM_User, time=0.8, value=0}
             <at 1 0>
            ]}
           RedCurve {keys [
             <at 0 0.5 RCIM_Cubic>
             {arriveTangentWeight=0.1342, interpMode=RCIM_Cubic, leaveTangentWeight=0.1342, tangentMode=RCTM_User, time=0.2, value=1}
             <at 0.2998 0.7>
             {arriveTangentWeight=0.1795, interpMode=RCIM_Cubic, leaveTangentWeight=0.1795, tangentMode=RCTM_User, time=0.4, value=0}
             <at 0.5 0>
             {arriveTangentWeight=0.0667, interpMode=RCIM_Cubic, leaveTangentWeight=0.0667, tangentMode=RCTM_User, time=0.6, value=0}
             <at 0.7 0>
             <at 0.8 0>
             <at 1 0>
            ]}
          }}
         CurveIndex=$Particles.SpectrumPosition
        ))
      ))
     B=&Waveform(
      Global Phase=&Add_Float(A=$Emitter.Age, B=$Particles.MaterialRandom)
      "Waveform [1]"=::NewEnumerator0
      "[1] Frequency"=2
      "[1] Compound Frequency"=3.17
      "[1] Amplitude Min/Max"=<0 0.1>
     )
    )) {isSetParameters=true, module=SetVariables_#6}
   /SetVariables(Particles.RibbonHeight:NiagaraFloat=&Add_Float(
     A=&Multiply_Float(A=$System.ModifiedAmplitudeData, B=0.1429)
     B=&Waveform(
      Global Phase=&Add_Float(A=$Emitter.Age, B=&RandomRangeFloat(Minimum=0, Maximum=1))
      "Waveform [1]"=::NewEnumerator0
      "[1] Frequency"=3
      "[1] Amplitude Min/Max"=&Lerp_Vector2(A=<0 0.003>, B=<-0.0714 0.0714>, Alpha=$System.ModifiedAmplitudeData)
      "[1] Bias"=&Lerp_Float(A=0.0357, B=0.0714, Alpha=$System.ModifiedAmplitudeData)
     )
    )) {isSetParameters=true}
   /SetVariables(Particles.RibbonPosition:NiagaraPosition=&ConvertVectorToPosition(
     Input Position=&MakeCustomVectorFromBool(
      Bool=&Boolean_AndOperation(
       Bool1=&SetBoolByFloatComparison(A=$Particles.SpectrumPosition, B=0, Comparison Type=::NewEnumerator5)
       Bool2=&SetBoolByFloatComparison(A=$Particles.SpectrumPosition, B=1, Comparison Type=::NewEnumerator5)
      )
      True Vector=&Add_Vector(
       A=&ConvertPositionToVector(Input Position=$Particles.Initial.Position, Convert To Absolute Large World Vector=false)
       B=&Multiply_VectorByFloat(Vector=$Particles.Initial.SpectrumUpVector, Float=&Multiply_Float(A=$Particles.RibbonHeight, B=28))
      )
      False Vector=&ConvertPositionToVector(Input Position=$Particles.Initial.Position, Convert To Absolute Large World Vector=false)
     )
     Vector Is In Absolute Large World Space:NiagaraBool=false
    )) {isSetParameters=true}
  ]
  rendererClasses [&NiagaraRibbonRendererProperties]
  renderers [{rendererClass=&NiagaraRibbonRendererProperties}]
 }
 SingleBand_LeftToRight {
  bEnabled=false
  simTarget=GPUComputeSim
  emitterUpdateScript [
   &EmitterState(Life Cycle Mode=::NewEnumerator1, Loop Behavior=::NewEnumerator1, Loop Duration Mode=::NewEnumerator1, UseLoopDelay=false)
   &SpawnBurst_Instantaneous(Spawn Count=39, Use Spawn Probability=false, UseLoopCountLimit=false)
  ]
  particleSpawnScript [
   /SetVariables(Particles.SpectrumPosition:NiagaraFloat=&ReturnNormalizedExecIndex()) {isSetParameters=true, module=SetVariables_#4}
   &InitializeParticle(
    Color Mode=::NewEnumerator0
    Lifetime=5
    Mesh Scale=<2 2 2>
    Mesh Scale Mode=::NewEnumerator4
    Position Offset=&VectorFromCurve(
     VectorCurve=DataInterface() {propertyValues {
       XCurve {keys [
         {arriveTangent=1.9015, interpMode=RCIM_Cubic, leaveTangent=1.9015, tangentMode=RCTM_User, time=0, value=0}
         <at 0.5 0.5 RCIM_Cubic>
         {arriveTangent=-1.9808, interpMode=RCIM_Cubic, leaveTangent=-1.9808, tangentMode=RCTM_User, time=1, value=0}
        ]}
       YCurve {keys [<at 0 0>,<at 1 0>]}
       ZCurve {keys [
         {
          arriveTangent=-0.0114
          arriveTangentWeight=0.4956
          interpMode=RCIM_Cubic
          leaveTangent=-0.0114
          leaveTangentWeight=0.4956
          tangentMode=RCTM_User
          tangentWeightMode=RCTWM_WeightedBoth
          time=0
          value=-1
         }
         {
          arriveTangent=-0.0012
          arriveTangentWeight=0.4997
          interpMode=RCIM_Cubic
          leaveTangent=-0.0012
          leaveTangentWeight=0.4997
          tangentMode=RCTM_User
          tangentWeightMode=RCTWM_WeightedBoth
          time=1
          value=1
         }
        ]}
      }}
     CurveIndex=$Particles.SpectrumPosition
     Scale Curve=<31 1 13.2>
    )
    Sprite Size Max=<3 50>
    Sprite Size Min=<3 50>
    Sprite Size Mode=::NewEnumerator2
    UsePositionOffset=true
    Write Material Random=true
   )
   /SetVariables(Particles.SpectrumUpVector:Vector3f=/Niagara/DynamicInputs/Transforms/TransformPosition(
     Position=&ConvertPositionToVector(Input Position=$Particles.Position, Convert To Absolute Large World Vector=false)
     Source Space=::Local
     Destination Space=::Simulation
    )) {isSetParameters=true, module=SetVariables_#8}
   &InitialMeshOrientation(Mesh Orientation Mode=::NewEnumerator2, Orientation Vector=$Particles.SpectrumUpVector, Use Rotation Vector=false)
   /SetVariables() {isSetParameters=true, module=SetVariables_#5}
   &AlignSpriteToMeshOrientation(Mesh Orientation Relative Sprite Alignment Vector=<1 0 0>, Mesh Orientation Relative Sprite Facing Vector:Vector3f=<-1 0 0>)
   /SetVariables() {isSetParameters=true, module=SetVariables_#10}
  ]
  particleUpdateScript [
   &ParticleState(
    Kill Particles When Lifetime Has Elapsed=false
    Let Infinitely Lived Particles Die When Emitter Deactivates=true
    Loop Particles Lifetime=true
   )
   &UpdateMeshOrientation(
    Rotation Rate=&Sine(Normalized Angle=&Add_Float(A=$Emitter.Age, B=$Particles.SpectrumPosition), Period=0.333, Scale=0.1, Bias=0)
    Rotation Vector=<0 0 1>
   )
   /SetVariables(Particles.SpectrumSample:NiagaraFloat=&MakeCustomFloatFromBool(
     Bool=&SetBoolByFloatComparison(A=$Particles.SpectrumPosition, B=0, Comparison Type=::NewEnumerator4)
     True Float=$System.ModifiedAmplitudeData
     False Float=&self:GetPreviousSpectrumSample(ReadAttribute:NiagaraDataInterfaceParticleRead=DataInterface() {propertyValues {EmitterBinding {bindingMode=Other, emitterName=SingleBand_LeftToRight}}})
    )) {isSetParameters=true, module=SetVariables_#6}
   &ScaleColor(
    Curve Index=$Particles.SpectrumSample
    Linear Color Curve=DataInterface() {propertyValues {
      AlphaCurve {keys [<at 0 1>]}
      BlueCurve {keys [<at 0 0>,<at 0.4775 1>,<at 1 0>]}
      GreenCurve {keys [<at 0 0>,<at 0.4775 0.227>,<at 1 1>]}
      RedCurve {keys [<at 0 1>,<at 0.4775 0>,<at 1 0.1772>]}
     }}
    Scale Mode=::NewEnumerator2
   )
   &ScaleSpriteSize(Scale Factor=&MakeVector2D(X=1, Y=$Particles.SpectrumSample), Scale Sprite Size Mode=::NewEnumerator2)
   &AlignSpriteToMeshOrientation(Mesh Orientation Relative Sprite Alignment Vector=<1 0 0>, Mesh Orientation Relative Sprite Facing Vector:Vector3f=<-1 0 0>) {module=AlignSpriteToMeshOrientation001}
   /SetVariables(Particles.PrevSpectrumSample:NiagaraFloat=$Particles.SpectrumSample) {isSetParameters=true, module=SetVariables_#11}
  ]
  rendererClasses [&NiagaraSpriteRendererProperties]
  renderers [{rendererClass=&NiagaraSpriteRendererProperties}]
 }
 SingleBand_Mirrored {
  bEnabled=false
  simTarget=GPUComputeSim
  emitterUpdateScript [
   &EmitterState(Life Cycle Mode=::NewEnumerator1, Loop Behavior=::NewEnumerator1, Loop Duration Mode=::NewEnumerator1, UseLoopDelay=false)
   &SpawnBurst_Instantaneous(Spawn Count=39, Use Spawn Probability=false, UseLoopCountLimit=false)
  ]
  particleSpawnScript [
   /SetVariables(Particles.SpectrumPosition:NiagaraFloat=&Divide_Float(
     A=&Ceil_Float(Float=&Multiply_Float(A=&ReturnNormalizedExecIndex(), B=19))
     B=19
    )) {isSetParameters=true, module=SetVariables_#4}
   /SetVariables(Particles.LeftRight:NiagaraFloat=&Multiply_Float(
     A=&Add_Float(
      A=&Frac_Float(Float=&Multiply_Float(A=&ReturnNormalizedExecIndex(), B=19))
      B=-0.25
     )
     B=4
    )) {isSetParameters=true}
   &InitializeParticle(
    Color Mode=::NewEnumerator0
    Lifetime=5
    Mesh Scale=<2 2 2>
    Mesh Scale Mode=::NewEnumerator4
    Position Offset=&MakeVector(
     X=&FloatFromCurve(
      FloatCurve=DataInterface() {propertyValues {Curve {keys [
          {arriveTangent=0.3159, interpMode=RCIM_Cubic, leaveTangent=0.3159, tangentMode=RCTM_User, time=0, value=1}
          {arriveTangent=1.3145, interpMode=RCIM_Cubic, leaveTangent=1.3145, tangentMode=RCTM_User, time=1, value=0}
         ]}}}
      CurveIndex=$Particles.SpectrumPosition
      Scale Curve=15
     )
     Y=0
     Z=&Multiply_Float(
      A=&Multiply_Float(A=$Particles.LeftRight, B=14)
      B=&FloatFromCurve(
       FloatCurve=DataInterface() {propertyValues {Curve {keys [
           {arriveTangent=2.4758, interpMode=RCIM_Cubic, leaveTangent=2.4758, tangentMode=RCTM_User, time=0, value=0}
           {interpMode=RCIM_Cubic, tangentMode=RCTM_User, time=1, value=1}
          ]}}}
       CurveIndex=$Particles.SpectrumPosition
      )
     )
    )
    Sprite Size Max=<3 50>
    Sprite Size Min=<3 50>
    Sprite Size Mode=::NewEnumerator2
    UsePositionOffset=true
    Write Material Random=true
   )
   /SetVariables(Particles.SpectrumUpVector:Vector3f=&TransformPosition(
     Position=&ConvertPositionToVector(Input Position=$Particles.Position, Convert To Absolute Large World Vector=false)
     Source Space=::Local
     Destination Space=::Simulation
    )) {isSetParameters=true, module=SetVariables_#8}
   &InitialMeshOrientation(Mesh Orientation Mode=::NewEnumerator2, Orientation Vector=$Particles.SpectrumUpVector, Use Rotation Vector=false)
   /SetVariables() {isSetParameters=true, module=SetVariables_#5}
   /SetVariables(Particles.PeakPosition:NiagaraPosition=$Particles.Position) {isSetParameters=true, module=SetVariables_#9}
   &AlignSpriteToMeshOrientation(Mesh Orientation Relative Sprite Alignment Vector=<1 0 0>, Mesh Orientation Relative Sprite Facing Vector:Vector3f=<-1 0 0>)
   /SetVariables() {isSetParameters=true, module=SetVariables_#10}
  ]
  particleUpdateScript [
   &ParticleState(Kill Particles When Lifetime Has Elapsed=false)
   &UpdateMeshOrientation(
    Rotation Rate=&Sine(Normalized Angle=&Add_Float(A=$Emitter.Age, B=$Particles.SpectrumPosition), Period=0.333, Scale=0.1, Bias=0)
    Rotation Vector=<0 0 1>
   )
   /SetVariables(Particles.SpectrumSample:NiagaraFloat=&MakeCustomFloatFromBool(
     Bool=&SetBoolByFloatComparison(A=$Particles.SpectrumPosition, B=0, Comparison Type=::NewEnumerator4)
     True Float=$System.ModifiedAmplitudeData
     False Float=&self:GetPreviousSpectrumSample(ReadAttribute:NiagaraDataInterfaceParticleRead=DataInterface() {propertyValues {EmitterBinding {bindingMode=Other, emitterName=SingleBand_Mirrored}}})
    )) {isSetParameters=true, module=SetVariables_#6}
   &ScaleColor(
    Curve Index=$Particles.SpectrumSample
    Linear Color Curve=DataInterface() {propertyValues {
      AlphaCurve {keys [<at 0 1>]}
      BlueCurve {keys [<at 0 0>,<at 0.4775 1>,<at 1 0>]}
      GreenCurve {keys [<at 0 0>,<at 0.4775 0.227>,<at 1 1>]}
      RedCurve {keys [<at 0 1>,<at 0.4775 0>,<at 1 0.1772>]}
     }}
    Scale Mode=::NewEnumerator2
   )
   &ScaleSpriteSize(Scale Factor=&MakeVector2D(X=1, Y=$Particles.SpectrumSample), Scale Sprite Size Mode=::NewEnumerator2)
   &AlignSpriteToMeshOrientation(Mesh Orientation Relative Sprite Alignment Vector=<1 0 0>, Mesh Orientation Relative Sprite Facing Vector:Vector3f=<-1 0 0>) {module=AlignSpriteToMeshOrientation001}
   /SetVariables(Particles.PrevSpectrumSample:NiagaraFloat=$Particles.SpectrumSample) {isSetParameters=true, module=SetVariables_#11}
  ]
  rendererClasses [&NiagaraSpriteRendererProperties]
  renderers [{rendererClass=&NiagaraSpriteRendererProperties}]
 }
}
summary {
 systemName=NS_Backpack_CactusDancer
 emitters [
  {emitterName=Ribbon, rendererClasses [&NiagaraRibbonRendererProperties]}
  {bEnabled=false, emitterName=SingleBand_Mirrored, simTarget=GPUComputeSim, rendererClasses [&NiagaraSpriteRendererProperties]}
  {
   bEnabled=false
   emitterName=SingleBand_LeftToRight
   simTarget=GPUComputeSim
   rendererClasses [&NiagaraSpriteRendererProperties]
  }
  {emitterName=Light, rendererClasses [&NiagaraLightRendererProperties]}
  {emitterName=4BandL, rendererClasses [&NiagaraSpriteRendererProperties]}
  {emitterName=4BandR, rendererClasses [&NiagaraSpriteRendererProperties]}
 ]
 userVariables [
  User.bIsFrontEnd {type=NiagaraBool}
  User.bIsFrontEndPreview {type=NiagaraBool}
 ]
}
