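# Material function for the Cactus Dancer backpack that paints animated eyes onto incoming material attributes, toggled by a static switch.
# A two-frame vertical flipbook is sampled on UV1, with its red channel serving as the eye mask for blending a tint color into BaseColor.
# A Time-driven Frac and Ceil pulse briefly flips to the second frame each cycle like a blink, and Select_Frame offsets the frame choice.
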MaterialFunction /BRCosmetics/Accessories/FORT_Backpacks/Backpack_M_MED_Cactus_Dancer/Materials/MF_CactusDancer_Backpack_Eyes
graph {nodes [
  n1 = NamedRerouteUsage()
  n2 = SetMaterialAttributes!(
   MaterialAttributes=@n1
   BaseColor=VectorParameter() {group=CactusDancer Eyes, parameterName=Eyes_Color, defaultValue=<rgba 0.03 0.03 0.03 1>}.RGB
  ) {attributeSetTypes [#1]}
  n3 = NamedRerouteUsage()
  FunctionOutput(None=StaticSwitchParameter(True=BlendMaterialAttributes(A=@n1, B=@n2, Alpha=@n3), False=NamedRerouteUsage()) {group=Skin FX, parameterName=Use CactusDancer Eyes})
  NamedRerouteDeclaration(None=FunctionInput() {inputName=MA, inputType=FunctionInput_MaterialAttributes}) {name=MA_In, nodeColor=<rgba 1 0.5 0 1>}
  n4 = TextureCoordinate() {coordinateIndex=1}
  n5 = ScalarParameter() {group=CactusDancer Eyes, parameterName=Select_Frame, sliderMax=2, sortPriority=0}
  n6 = Divide(A=Time()) {constB=2.333}
  n7 = Ceil(None=Add(A=Frac(None=@n6)) {constB=-0.85})
  n8 = Add(A=ComponentMask(None=@n4) {g=true}, B=Saturate(None=Add(A=@n5, B=@n7))) {constB=0}
  n9 = TextureSampleParameter2D(UVs=AppendVector(A=ComponentMask(None=@n4) {r=true}, B=Divide(A=@n8))) {
   group=CactusDancer Eyes
   parameterName=Flipbook_Eyes
   samplerSource=SSM_Clamp_WorldGroupSettings
   samplerType=SAMPLERTYPE_LinearColor
   sortPriority=0
   texture=/BRCosmetics/Accessories/FORT_Backpacks/Backpack_M_MED_Cactus_Dancer/Textures/T_M_MED_Cactus_Dancer_Backpack_Eyes
  }
  NamedRerouteDeclaration(None=@n9.R) {name=A, nodeColor=<rgba 1 0.7734 0 1>}
 ]}
