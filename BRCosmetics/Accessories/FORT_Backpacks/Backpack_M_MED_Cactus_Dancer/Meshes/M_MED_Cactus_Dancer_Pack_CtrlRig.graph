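# Control Rig for the Cactus Dancer backpack cosmetic that adds procedural wobble to the pack's bone chain.
# The forward solve runs a single Chain Harmonics unit rooted at the pack bone, which parents the dyn_head and dyn_headphones bones.
# Only the wave is enabled, combining noise and bounce easing on X and Y, while pendulum and reach are off.
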ControlRigBlueprint /BRCosmetics/Accessories/FORT_Backpacks/Backpack_M_MED_Cactus_Dancer/Meshes/M_MED_Cactus_Dancer_Pack_CtrlRig
graphs [
 RigVMFunctionLibrary {links [], nodes []}
 RigVMModel {nodes [
   RigUnit_BeginExecution = &RigUnit_BeginExecution()
   /Script/ControlRig.RigUnit_ChainHarmonics(
    ChainRoot=pack
    DrawWorldOffset=(Rotation=(X=0,Y=0,Z=0,W=1),Translation=(X=0,Y=0,Z=0),Scale3D=(X=1,Y=1,Z=1))
    Pendulum=(bEnabled=False,PendulumStiffness=2,PendulumGravity=(X=0,Y=0,Z=0),PendulumBlend=0.75,PendulumDrag=0.9,PendulumMinimum=0,PendulumMaximum=1,PendulumEase=Linear,UnwindAxis=(X=0,Y=1,Z=0),UnwindMinimum=0.2,UnwindMaximum=0.05)
    Reach=(bEnabled=false,ReachTarget=(X=0,Y=0,Z=0),ReachAxis=(X=0,Y=0,Z=0),ReachMinimum=0,ReachMaximum=0,ReachEase=SineEaseInOut)
    Speed=(X=0.4,Y=0.45,Z=0)
    Wave=(bEnabled=True,WaveFrequency=(X=0.15,Y=0.2,Z=0),WaveAmplitude=(X=0.18,Y=0.1,Z=0),WaveOffset=(X=0,Y=0,Z=0),WaveNoise=(X=7,Y=6.5,Z=0),WaveMinimum=0,WaveMaximum=1,WaveEase=BounceEaseInOut)
    WaveCurve=(EditorCurveData=(Keys=((),(Time=1,Value=1))))
    bDrawDebug=false
    ExecutePin=@RigUnit_BeginExecution.ExecutePin
   ) {is=struct RigVMUnitNode}
  ]}
]
hierarchy [
 root {loc=<0 0 0>, rot=<0 0 0 1>}
 pack {parent=root, loc=<0 -0.7316 0>, rot=<0 0 0 1>}
 dyn_head {parent=pack, loc=<0.004 -10.2644 -3.825>, rot=<0.1075 0.6989 -0.1075 -0.6989>}
 dyn_headphones {parent=pack, loc=<0 -11.248 -4.1884>, rot=<0.1075 0.6989 -0.1075 -0.6989>}
]
