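# Post-process style AnimBP for the Cactus Dancer backpack, making its head bob to music during emotes.
# A root-offset input pose feeds spring Anim Dynamics on dyn_head outside emotes and a Control Rig during emotes, then is cached.
# Update reads the audio-reactive MPC spectrum, driving interpolated additive head yaw, with separate strengths selected by bIsInFrontEnd.
# The Ambient Music rig branch never reaches the output.
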AnimBlueprint /BRCosmetics/Accessories/FORT_Backpacks/Backpack_M_MED_Cactus_Dancer/Meshes/M_MED_Cactus_Dancer_Pack_AnimBP
target_skeleton=&/M_MED_Cactus_Dancer_Pack_Skeleton
graphs [
 AnimGraph {nodes [
   n1 = Get Ambient Music() -> (Ambient Music:double)
   n2 = Control Rig(Alpha=@n1) {
    alphaBoolBlend {blendOption=Linear}
    controlRigAssetReference {blueprintRigClass=&/M_MED_Cactus_Dancer_Pack_CtrlRig.M_MED_Cactus_Dancer_Pack_CtrlRig_C}
   }
   "Transform (Modify) Bone - Bone: dyn_head"(ComponentPose=Local To Component(LocalPose=@n2), Alpha=@n1, Translation="0,0,0", Rotation="0,35,0", Scale="1,1,1") {
    rotationMode=BMM_Additive
    rotationSpace=BCS_BoneSpace
    scaleMode=BMM_Additive
    scaleSpace=BCS_BoneSpace
    translationMode=BMM_Additive
    translationSpace=BCS_BoneSpace
    alphaBoolBlend {blendOption=Linear}
    alphaScaleBiasClamp {bInterpResult=true, interpSpeedDecreasing=20, interpSpeedIncreasing=45}
    boneToModify {boneName=dyn_head}
    rotation {yaw=35}
   }
   n3 = "Transform (Modify) Bone - Bone: dyn_head"(
    ComponentPose=Local To Component(LocalPose=Use cached pose 'Pose'())
    Alpha=Get Music Down() -> (Music Down:double)
    Translation="0,0,0"
    Rotation="0,-35,0"
    Scale="1,1,1"
   ) {
    rotationMode=BMM_Additive
    rotationSpace=BCS_BoneSpace
    scaleMode=BMM_Additive
    scaleSpace=BCS_BoneSpace
    translationMode=BMM_Additive
    translationSpace=BCS_BoneSpace
    alphaBoolBlend {blendOption=Linear}
    alphaScaleBiasClamp {bInterpResult=true, interpSpeedDecreasing=8, interpSpeedIncreasing=20}
    boneToModify {boneName=dyn_head}
    rotation {yaw=-35}
   }
   n4 = "Transform (Modify) Bone - Bone: dyn_head"(ComponentPose=@n3, Alpha=Get Music Up() -> (Music Up:double), Translation="0,0,0", Rotation="0,55,0", Scale="1,1,1") {
    rotationMode=BMM_Additive
    rotationSpace=BCS_BoneSpace
    scaleMode=BMM_Additive
    scaleSpace=BCS_BoneSpace
    translationMode=BMM_Additive
    translationSpace=BCS_BoneSpace
    alphaBoolBlend {blendOption=Linear}
    alphaScaleBiasClamp {bInterpResult=true, interpSpeedDecreasing=20, interpSpeedIncreasing=8}
    boneToModify {boneName=dyn_head}
    rotation {yaw=55}
   }
   n5 = "Transform (Modify) Bone - Bone: dyn_head"(
    ComponentPose=Local To Component(LocalPose=Use cached pose 'Pose'())
    Alpha=Get Music Down()
    Translation="0,0,0"
    Rotation="0,-30,0"
    Scale="1,1,1"
   ) {
    rotationMode=BMM_Additive
    rotationSpace=BCS_BoneSpace
    scaleMode=BMM_Additive
    scaleSpace=BCS_BoneSpace
    translationMode=BMM_Additive
    translationSpace=BCS_BoneSpace
    alphaBoolBlend {blendOption=Linear}
    alphaScaleBiasClamp {bInterpResult=true, interpSpeedDecreasing=5, interpSpeedIncreasing=7}
    boneToModify {boneName=dyn_head}
    rotation {yaw=-30}
   }
   n6 = "Transform (Modify) Bone - Bone: dyn_head"(ComponentPose=@n5, Alpha=Get Music Up(), Translation="0,0,0", Rotation="0,38,0", Scale="1,1,1") {
    rotationMode=BMM_Additive
    rotationSpace=BCS_BoneSpace
    scaleMode=BMM_Additive
    scaleSpace=BCS_BoneSpace
    translationMode=BMM_Additive
    translationSpace=BCS_BoneSpace
    alphaBoolBlend {blendOption=Linear}
    alphaScaleBiasClamp {bInterpResult=true, interpSpeedDecreasing=7, interpSpeedIncreasing=5}
    boneToModify {boneName=dyn_head}
    rotation {yaw=38}
   }
   Output Pose(Result=Blend Poses by bool(BlendPose_0=Component To Local(ComponentPose=@n4), BlendPose_1=Component To Local(ComponentPose=@n6), bActiveValue=Get bIsInFrontEnd()) {blendType=Linear}) {layerGroup=None}
   n7 = "Transform (Modify) Bone - Bone: root"(ComponentPose=Local To Component(LocalPose=Input Pose()), Translation="0,0.5,0", Rotation="0,0,0") {translationMode=BMM_Additive, alphaBoolBlend {blendOption=Linear}, boneToModify {boneName=root}, translation {y=0.5}}
   n8 = Anim Dynamics - Bone: dyn_head(ComponentPose=@n7, Alpha=Get Music_alpha() -> (Music_alpha:double), ExternalForce="0,0,0") {
    angularDampingOverride=0.9
    angularSpringConstant=150
    bAngularSpring=true
    bGravityOverrideInSimSpace=true
    bLinearSpring=true
    bOverrideAngularDamping=true
    bOverrideLinearDamping=true
    gravityScale=0.15
    linearDampingOverride=0.995
    linearSpringConstant=75
    simulationSpace=BoneRelative
    alphaBoolBlend {blendOption=Linear}
    alphaScaleBiasClamp {bClampResult=true, bInterpResult=true, clampMax=0.6, interpSpeedDecreasing=7, interpSpeedIncreasing=5}
    boundBone {boneName=dyn_head}
    componentLinearAccScale=<0.05 0.1 0.1>
    componentLinearVelScale=<0.6 0.6 0.6>
    gravityOverride {z=-980}
    physicsBodyDefinitions [{
      sphereCollisionRadius=0
      boundBone {boneName=dyn_head}
      boxExtents=<250 250 250>
      constraintSetup {angularLimitsMax=<0 0 0>, angularLimitsMin=<0 0 0>, angularTarget=<0 0 0>, linearAxesMax=<3 0.5 0.5>, linearAxesMin=<-4 -0.5 -0.5>}
      localJointOffset=<-3 0 0>
     }]
    relativeSpaceBone {boneName=root}
   }
   n9 = Control Rig!(Source=Component To Local(ComponentPose=@n8), bAlphaBoolEnabled:bool=Get bIsPlayingEmote()) -> (Pose) {
    alphaInputType=Bool
    alphaBoolBlend {blendOption=Linear}
    controlRigAssetReference {blueprintRigClass=&M_MED_Cactus_Dancer_Pack_CtrlRig_C}
   }
   Save cached pose 'Pose'(Pose=@n9) {cacheName=Pose}
  ]}
 EventGraph {nodes [
   n1 = GetVectorParameterValue!(
    execute
    Collection:Material Parameter Collection Object Reference=/Game/Athena/Events/AudioAnalysis/MPC_Audio_Reactive_Parameters
    ParameterName:Name=Spectrum_Music
   ) -> (then, ReturnValue_R:float=~, ReturnValue_G:float=~, ReturnValue_B:float=~, ReturnValue_A:float=~) {is=title K2Node_CallMaterialParameterCollectionFunction}
   n2 = float * float(A="Sin (Radians)"(A=@n1.ReturnValue_A), B=50) {is=title K2Node_PromotableOperator}
   > Set Music Down(Music Down:double=SelectFloat(A=MapRangeClamped(Value=@n2, InRangeA=18, InRangeB=10), B=0, bPickA=Get bIsPlayingEmote())) -> (Output_Get:double)
   n3 = GetScalarParameterValue!(
    execute
    Collection:Material Parameter Collection Object Reference=&MPC_Audio_Reactive_Parameters
    ParameterName:Name=Amplitude_Music
   ) -> (then, ReturnValue:float) {is=title K2Node_CallMaterialParameterCollectionFunction}
   n4 = float * float(A=@n3, B=50)
   > Set Music Up_Scalar(Music Up_Scalar:double=MapRangeClamped(Value=@n4, InRangeA=25, InRangeB=40)) -> (Output_Get:double)
   > Set Music Down_Scalar(Music Down_Scalar:double=MapRangeClamped(Value=@n4, InRangeA=25, InRangeB=1)) -> (Output_Get:double)
   n5 = GetVectorParameterValue(Collection=&MPC_Audio_Reactive_Parameters, ParameterName=Spectrum_Music) -> (ReturnValue_R=~, ReturnValue_G=~, ReturnValue_B=~, ReturnValue_A=~)
   n6 = float * float(A="Sin (Radians)"(A=@n5.ReturnValue_A), B=50)
   > Set Music Up(Music Up:double=SelectFloat(A=MapRangeClamped(Value=@n6, InRangeA=20, InRangeB=30), B=0, bPickA=Get bIsPlayingEmote())) -> (Output_Get:double)
   Event BlueprintUpdateAnimation()
   > n7 = Sequence!(execute) -> (then_0, then_1, then_2:Exec, then_3:Exec)
   > n8 = GetVectorParameterValue(Collection=&MPC_Audio_Reactive_Parameters, ParameterName=Spectrum_Music) -> (ReturnValue_R=~, ReturnValue_G=~, ReturnValue_B=~, ReturnValue_A=~)
   n9 = float * float(A="Sin (Radians)"(A=@n8), B=50)
   > Set Music Up(Music Up=SelectFloat(A=MapRangeClamped(Value=@n9, InRangeA=20, InRangeB=35), B=0, bPickA=Get bIsPlayingEmote()))
   n10 = GetVectorParameterValue(execute=@n7.then_1, Collection=&MPC_Audio_Reactive_Parameters, ParameterName=Spectrum_Music) -> (ReturnValue_R=~, ReturnValue_G=~, ReturnValue_B=~, ReturnValue_A=~)
   n11 = float * float(A="Sin (Radians)"(A=@n10.ReturnValue_A), B=50)
   > Set Music Down(Music Down=SelectFloat(A=MapRangeClamped(Value=@n11, InRangeA=25, InRangeB=10), B=0, bPickA=Get bIsPlayingEmote()))
   n12 = Branch(execute=@n7.then_2, Condition=Get bIsPlayingEmote())
   > Set Music_alpha(Music_alpha:double=0) -> (Output_Get:double)
   Set Music_alpha(execute=@n12.else, Music_alpha=1)
   n13 = GetVectorParameterValue(execute=@n7.then_3, Collection=&MPC_Audio_Reactive_Parameters, ParameterName=Spectrum_Music) -> (ReturnValue_R=~, ReturnValue_G=~, ReturnValue_B=~, ReturnValue_A=~)
   > n14 = Branch(Condition=Get bIsPlayingEmote())
   > Set Ambient Music(Ambient Music:double=0) -> (Output_Get:double)
   n15 = "Sin (Radians)"(A=@n13.ReturnValue_G)
   Set Ambient Music(execute=@n14.else, Ambient Music=MapRangeClamped(Value=float * float(A=@n15, B=50), InRangeA=5, InRangeB=35))
   TryGetPawnOwner()
   Branch()
   Set Music Down(Music Down=0)
  ]}
]
variables [
 Ambient Music {type=number}
 Body_Mesh {type=SkeletalMeshComponent}
 Headbop {type=Rotator}
 Music Down {type=number}
 Music Down_Scalar {type=number}
 Music Up {type=number}
 Music Up_Scalar {type=number}
 Music_alpha {type=number}
]
